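# One-shot mono MetaSound source for the bounties UI hover sound in the Bounties/Rivalry/Duels gameplay set, routed to the Fort_UI sound class.
# On play, a wave player plays the default hover wave, its audio is multiplied by a master volume input, and wave completion signals finish.
# A concurrency override allows a single instance, stopping the oldest with volume scaling, and virtualization is disabled.
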MetaSoundSource /Game/Sounds/Fort_UI_Sounds/Gameplay/Bounties_Rivalry_Duels/MSS_UI_Bounties_Hover
inputs [
 Default Hover Sound {type=WaveAsset}
 MasterVolume {type=Float}
 UE.Source.OnPlay {type=Trigger}
]
interfaces [UE.OutputFormat.Mono:1.0, UE.Source.OneShot:1.0, UE.Source:1.1]
outputs [
 "UE.OutputFormat.Mono.Audio:0" {type=Audio}
 UE.Source.OneShot.OnFinished {type=Trigger}
]
pages [Default {
  input_defaults {
   Default Hover Sound=/Game/Sounds/Fort_UI_Sounds/Gameplay/Bounties_Rivalry_Duels/Wavs/UI_Bounties_Hover
   MasterVolume=0.7079
   UE.Source.OnPlay=null
  }
  nodes [
   WavePlayer_0 = UE.Wave Player:Mono(Wave Asset=^Default Hover Sound, Play=^UE.Source.OnPlay) {also {version [1,1]}}
   Multiply_0 = UE.Multiply:Audio by Float(AdditionalOperands=^MasterVolume, PrimaryOperand=@WavePlayer_0.Out Mono)
  ]
  links [{dst="@UE.OutputFormat.Mono.Audio:0", src=Multiply_0.Out}, {dst="@UE.Source.OneShot.OnFinished", src=WavePlayer_0.On Finished}]
 }]
properties {
 assetImportData=&self:AssetImportData
 bOverrideConcurrency=true
 priority=0.75
 sampleRate=48000
 soundClassObject=&Fort_UI
 virtualizationMode=Disabled
 concurrencyOverrides {bVolumeScaleCanRelease=true, maxCount=1, resolutionRule=StopOldest, volumeScale=0.5, volumeScaleAttackTime=0.25, volumeScaleReleaseTime=0.25}
}
